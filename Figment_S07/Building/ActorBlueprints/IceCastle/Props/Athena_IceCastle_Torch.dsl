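; Ice castle torch prop actor blueprint with a fire particle effect and looping torch audio.
; When the replicated Destroyed flag becomes true, its OnRep handler deactivates the fire particle component and stops the audio component.
; The construction script is empty, and no other logic is present.
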(fn OnRep_Destroyed
  (if (Variables|Default|GetDestroyed)
    (Components|Activation|Deactivate (Variables|Default|GetP_Torch_Fire_01))
    (Audio|Components|Audio|Stop (Variables|Default|GetTorchAudio))))

(fn ConstructionScript)
