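; Ice Castle torch prop that goes out permanently when damaged.
; Server damage of any amount sets the Destroyed flag, whose OnRep deactivates the fire particles, stops the looping fire audio and destroys the point light.
; The damage event also calls DisableLight, which destroys the point light only on the Remote branch of its authority switch.
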(event EventOnDamageServer
  (params (Damage :type Float%20%28single-precision%29) (DamageTags :type Gameplay%20Tag%20Container%20Structure%20%28by%20ref%29) (Momentum :type Vector :default "0, 0, 0") (HitInfo :type Hit%20Result%20Structure%20%28by%20ref%29) (InstigatedBy :type Controller%20Object%20Reference) (DamageCauser :type Actor%20Object%20Reference) (EffectContext :type Gameplay%20Effect%20Context%20Handle%20Structure))
  (Variables|Default|SetDestroyed true)
  (CallFunction|DisableLight))

(event Custom|DisableLight
  (Networking|SwitchHasAuthority
    (:Remote
      (Components|DestroyComponent (Variables|Default|GetPointLight)))))

(fn OnRep_Destroyed
  (if (Variables|Default|GetDestroyed)
    (Components|Activation|Deactivate (Variables|Default|GetP_Torch_Fire_01))
    (Audio|Components|Audio|Stop (Variables|Default|GetElemental_FireLoop_Small_Cue))
    (Components|DestroyComponent (Variables|Default|GetPointLight))))

(fn ConstructionScript)
